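; Cosmos AI pawn Blueprint that plays directional audio tells when its alert level changes and sets its death loot tier to Loot_Cosmos.
; On the client, tells play only when the closest listener lies within a forward cone, tested by dot product.
; Suspicious plays the alerted sound and re-arms the threatened sound, which then plays once on Threatened.
; A distance-indexed ScalableFloat sets each sound's short-range "Range" parameter.
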(event EventBeginPlay
  (|CallParentFunction) ; This node is disabled and will not be called.\nDrag off pins to build functionality.
)

(event Collision|EventActorBeginOverlap
  (params (OtherActor :type Actor%20Object%20Reference))
  (|CallParentFunction OtherActor) ; This node is disabled and will not be called.\nDrag off pins to build functionality.
)

(event EventTick
  (params (DeltaSeconds :type Float%20%28single-precision%29))
  (|CallParentFunction DeltaSeconds) ; This node is disabled and will not be called.\nDrag off pins to build functionality.
)

(fn ClientOnAlertLevelChanged
  (params (OldAlertLevel :type EAlertLevel%20Enum :default "Unaware") (NewAlertLevel :type EAlertLevel%20Enum :default "Unaware"))
  (bind _returnvalue (Transformation|GetActorLocation))
  (|CallParentFunction OldAlertLevel NewAlertLevel)
  (bind _listenerposition (Audio|GetClosestListenerLocation _returnvalue 2.5e+05 true))
  (if (CallFunction|PassesAudioTellCheck _listenerposition _returnvalue)
    (if (Utilities|Enum|Equal(Enum) NewAlertLevel "Threatened")
      (if (Variables|Default|GetbHasResetBackToAlerted)
        (bind _returnvalue_2 (Audio|SpawnSoundatLocation (Variables|Default|GetOnThreatenedAudioTell) (Transformation|GetActorLocation)))
        (Audio|Parameter|SetIntegerParameter _returnvalue_2 "Range" (Math|Conversions|ToInteger(Boolean) (.ShortRangeCheck (CallFunction|PassesAudioTellCheck _listenerposition _returnvalue))) :!class /Script/AudioExtensions.AudioParameterControllerInterface)
        (Variables|Default|SetbHasResetBackToAlerted false))
      (elif (Utilities|Enum|Equal(Enum) NewAlertLevel "Suspicious")
        (bind _returnvalue_3 (Audio|SpawnSoundatLocation (Variables|Default|GetOnAlertedAudioTell) (Transformation|GetActorLocation)))
        (Audio|Parameter|SetIntegerParameter _returnvalue_3 "Range" (Math|Conversions|ToInteger(Boolean) (.ShortRangeCheck (CallFunction|PassesAudioTellCheck _listenerposition _returnvalue))) :!class /Script/AudioExtensions.AudioParameterControllerInterface)
        (Variables|Default|SetbHasResetBackToAlerted true)))))

(fn PassesAudioTellCheck
  (params (Listener :type Vector) (AI :type Vector))
  (return (>= (Math|Vector|DotProduct (Math|Vector|Normalize (- Listener AI)) (Transformation|GetActorForwardVector)) 0.8) (> (ScalableFloat|GetValueatLevel (Variables|Default|GetShortRangeThreshold) (Math|Vector|VectorLength (- Listener AI))) 1.0)))

(fn SetupLootTierNameWithWeapon
  (Variables|Loot|SetLootTierNameToSpawnWhenDead "Loot_Cosmos")
  (return))

(fn ConstructionScript
  (|CallParentFunction))
